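; Suitcase of Patience, an interactable prop that a player opens over a configurable OpenTimeSeconds duration.
; Interaction is allowed only while AllowInteract is true; the code that changes that flag is not in these graphs.
; Its OnRep function swaps between closed and searched suitcase meshes, showing the Coral1 component only in the searched state.
; The construction script sets the same two looks from a DebugOpen flag for previewing.
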(fn BlueprintCanInteract
  (params (InteractingPawn :type Fort%20Pawn%20Object%20Reference) (InteractionBeingAttempted :type EInteractionBeingAttempted%20Enum :default "FirstInteraction") (InteractionType :type TInteractionType%20Enum :default "IT_NoInteraction"))
  (return (Variables|Default|GetAllowInteract)))

(fn BlueprintGetInteractionString
  (params (InteractingPawn :type Fort%20Pawn%20Object%20Reference) (InteractionBeingAttempted :type EInteractionBeingAttempted%20Enum :default "FirstInteraction"))
  (return "NSLOCTEXT(\"[F4F9C1BE77CAC6F95EAFF0D9633EAEFA]\", \"605CBB6C4249E041F91582B5A0F137D3\", \"Open Suitcase of Patience\")"))

(fn BlueprintGetInteractionTime
  (params (InteractingPawn :type Fort%20Pawn%20Object%20Reference) (InteractionBeingAttempted :type EInteractionBeingAttempted%20Enum :default "FirstInteraction"))
  (return (Variables|Default|GetOpenTimeSeconds) true))

(fn OnRep_AllowInteract
  (bind _coral1 (Variables|Default|GetCoral1))
  (bind _staticmesh (Variables|Default|GetStaticMesh))
  (bind _coral1_2 (Variables|Default|GetCoral1))
  (if (Variables|Default|GetAllowInteract)
    (Components|StaticMesh|SetStaticMesh _staticmesh "/Game/Environments/Props/Boxes/Meshes/Prop_LargeSuitcases_01.Prop_LargeSuitcases_01")
    (Development|SetHiddeninGame _coral1 true false)
    (Rendering|SetVisibility _coral1 false false)
    (else
      (Components|StaticMesh|SetStaticMesh _staticmesh "/Game/Environments/Props/Boxes/Meshes/Prop_LargeSuitcases_01_Searched.Prop_LargeSuitcases_01_Searched")
      (Development|SetHiddeninGame _coral1_2 false false)
      (Rendering|SetVisibility _coral1_2 true false))))

(fn ConstructionScript
  (bind _coral1 (Variables|Default|GetCoral1))
  (bind _staticmesh (Variables|Default|GetStaticMesh))
  (bind _coral1_2 (Variables|Default|GetCoral1))
  (if (Variables|Default|GetDebugOpen)
    (Components|StaticMesh|SetStaticMesh _staticmesh "/Game/Environments/Props/Boxes/Meshes/Prop_LargeSuitcases_01_Searched.Prop_LargeSuitcases_01_Searched")
    (Development|SetHiddeninGame _coral1_2 false false)
    (Rendering|SetVisibility _coral1_2 true false)
    (else
      (Components|StaticMesh|SetStaticMesh _staticmesh "/Game/Environments/Props/Boxes/Meshes/Prop_LargeSuitcases_01.Prop_LargeSuitcases_01")
      (Development|SetHiddeninGame _coral1 true false)
      (Rendering|SetVisibility _coral1 false false))))
